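# Translucent unlit Niagara mesh-particle material for motion-blur streak FX on the Tilted Parrot glider.
# Color comes from a low-mip Sunrise Castle glider diffuse, desaturated, tinted by particle color, fake-lit by vertex normal, with panning noise highlights.
# Opacity combines noise-distorted and scrolling streaks with end fades, inverse Fresnel and a facing mask, using reflection-vector cone UVs.
# PreviousFrameSwitch fakes velocity for motion blur, with front-end strength selected by dynamic parameter.
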Material /BRCosmetics/Gadgets/Assets/VinderTech_GliderChute/Glider_TiltedParrot/FX/M_FX_Glider_TiltedParrot
graph {
 properties {
  allowTranslucentCustomDepthWrites=true
  bOutputTranslucentVelocity=true
  bUsedWithNiagaraMeshParticles=true
  blendMode=BLEND_Translucent
  opacityMaskClipValue=0
  pixelDepthOffsetMode=PDOM_Legacy
  shadingModel=MSM_Unlit
  usedShadingModels=Unlit
 }
 nodes [
  Constant() {r=1}
  LinearInterpolate(
   A=ScalarParameter() {defaultValue=0.8, parameterName=Desaturation_Start}
   B=ScalarParameter() {defaultValue=-0.75, parameterName=Desaturation_End}
   Alpha=ComponentMask(None=TextureCoordinate()) {g=true}
  )
  n1 = NamedRerouteUsage()
  n2 = VectorParameter() {parameterName=NoiseScale, defaultValue=<rgba 0 0 -4 -3>}
  n3 = ComponentMask(None=@n2.RGB) {g=true, r=true}
  n4 = Time()
  n5 = DynamicParameter() {paramNames [Param1,Param2,IsFrontEnd,Param4]}
  n6 = TextureSample(UVs=Panner(Coordinate=Multiply(A=@n1, B=@n3), Time=Add(A=@n4, B=@n5.Param2), Speed=AppendVector(A=@n2.B, B=@n2.B)) {speedY=-3}) {samplerType=SAMPLERTYPE_Grayscale, texture=&LowResBlurredNoise}
  n7 = Multiply(
   A=Power(Base=@n6.B, Exp=ScalarParameter() {defaultValue=4.5, parameterName=HighlightsCtrst})
   B=ScalarParameter() {defaultValue=3, parameterName=HighlightsStr}
  )
  n8 = NamedRerouteUsage()
  n9 = TextureSample!(UVs=TextureCoordinate() {coordinateIndex=1}, Tex, Level, Apply View MipBias) {
   constMipValue=4
   mipValueMode=TMVM_MipLevel
   texture=/BRCosmetics/Gadgets/Assets/VinderTech_GliderChute/Glider_Sunrise_Castle_Male/Textures/T_Glider_Sunrise_Castle_Male_D
  }
  n10 = ParticleColor()
  n11 = Multiply(A=Desaturation(None=@n9.RGB), B=@n10.RGB)
  n12 = DotProduct(A=VertexNormalWS(), B=Constant3Vector() {constant=<rgba 0 0 1 0>})
  n13 = ScalarParameter() {defaultValue=1.5, parameterName=LightingStr}
  n14 = Add(A=@n11, B=Multiply(A=ConstantBiasScale(None=@n12), B=Multiply(A=@n11, B=@n13)))
  MaterialFunctionCall(Emissive=Add(A=Multiply(A=@n7, B=@n8), B=NamedRerouteDeclaration(None=@n14) {name=ResolvedColor, nodeColor=<rgba 0 1 0.6484 1>})) {materialFunction=&MF_FX_EmissiveAdjustments}
  n15 = NamedRerouteUsage()
  n16 = ComponentMask(None=VectorParameter() {parameterName=HighlightDist, defaultValue=<rgba 0 0.06 0 1>}.RGB) {g=true, r=true}
  n17 = TextureSample!(UVs=Add(A=NamedRerouteUsage(), B=Multiply(A=@n15, B=@n16)), Tex, Apply View MipBias) {
   samplerSource=SSM_Clamp_WorldGroupSettings
   samplerType=SAMPLERTYPE_LinearColor
   texture=/BRCosmetics/Gadgets/Assets/VinderTech_GliderChute/Glider_Sunrise_Castle_Male/FX/T_FX_SunriseCastleMale_MotionBlurStreaks
  }
  n18 = Time()
  n19 = DynamicParameter() {paramNames [Param1,Param2,IsFrontEnd,Param4]}
  n20 = TextureSample(UVs=Panner(Coordinate=NamedRerouteUsage(), Time=Add(A=@n18, B=@n19.Param2)) {speedY=6}) {
   samplerSource=SSM_Wrap_WorldGroupSettings
   samplerType=SAMPLERTYPE_LinearColor
   texture=&T_FX_SunriseCastleMale_MotionBlurStreaks
  }
  n21 = Add(
   A=Multiply(A=@n17.R, B=ScalarParameter() {defaultValue=0.8, parameterName=MaskStr})
   B=Multiply(A=@n20.G, B=ScalarParameter() {defaultValue=0.65, parameterName=BarStr})
  )
  n22 = Multiply(A=Multiply(A=@n10.A, B=Saturate(None=@n21)))
  n23 = NamedRerouteUsage()
  n24 = ComponentMask(None=TextureCoordinate()) {g=true}
  n25 = Add(A=@n24, B=ScalarParameter() {group=Alpha, parameterName=Side1 Length Fade Bias, sortPriority=2})
  n26 = ScalarParameter() {defaultValue=5, group=Alpha, parameterName=Side1 Length Fade Scale}
  n27 = LinearInterpolate(
   B=Saturate(None=Multiply(A=@n25, B=@n26) {constB=3})
   Alpha=ScalarParameter() {defaultValue=1, group=Alpha, parameterName=Side1 Fade Length Edge, sortPriority=7}
  ) {constA=1}
  n28 = Add(A=OneMinus(None=@n24), B=ScalarParameter() {group=Alpha, parameterName=Side2 Length Fade Bias, sortPriority=3})
  n29 = ScalarParameter() {defaultValue=2, group=Alpha, parameterName=Side2 Length Fade Scale, sortPriority=1}
  n30 = LinearInterpolate(
   B=Saturate(None=Multiply(A=@n28, B=@n29) {constB=2})
   Alpha=ScalarParameter() {defaultValue=1, group=Alpha, parameterName=Side2 Fade Length Edge, sortPriority=8}
  ) {constA=1}
  n31 = ScalarParameter() {defaultValue=2, parameterName=FresnelExp}
  n32 = ScalarParameter() {parameterName=BaseReflectFrac}
  Multiply(
   A=Multiply(A=Multiply(A=@n22, B=@n23), B=Multiply(A=@n27, B=@n30))
   B=OneMinus(None=Fresnel(ExponentIn=@n31, BaseReflectFractionIn=@n32) {exponent=1})
  )
  n33 = Subtract(A=CameraPositionWS(), B=WorldPosition().XYZ)
  n34 = Transform(None=MaterialFunctionCall!(Normal=VertexNormalWS(), CameraVector=Normalize(VectorInput=@n33)) {materialFunction=/Engine/Functions/Engine_MaterialFunctions02/Utility/CustomReflectionVector}.Reflection Vector) {transformSourceType=TRANSFORMSOURCE_World, transformType=TRANSFORM_Tangent}
  n35 = ComponentMask(None=@n34) {b=true}
  n36 = Constant() {r=8}
  n37 = SquareRoot(None=Multiply(A=Add(A=@n35, B=@n36), B=Constant() {r=0.8})) {desc=Z Component Square}
  n38 = Add(A=Divide(A=ComponentMask(None=@n34) {desc=X Component, r=true}, B=@n37), B=Constant() {r=0.5})
  n39 = AppendVector(A=@n38, B=ComponentMask(None=TextureCoordinate()) {g=true})
  NamedRerouteDeclaration(None=MaterialFunctionCall(XY=VertexInterpolator(VS=@n39).PS) {materialFunction=&Swizzle}.YX) {name=ConeUV, nodeColor=<rgba 0.7578 0 1 1>}
  NamedRerouteDeclaration(None=Subtract(A=Multiply(A=@n6.B) {constB=2})) {name=HighlightNoise, nodeColor=<rgba 1 0 0.7266 1>}
  n40 = ComponentMask(None=@n34) {desc=Y Component, g=true}
  n41 = AppendVector(A=@n38, B=Add(A=Constant() {r=0.5}, B=Divide(A=@n40, B=@n37)))
  n42 = ScalarParameter() {defaultValue=1, group=SideOn, parameterName=FacingMask_H}
  n43 = ComponentMask(None=TextureCoordinate()) {g=true}
  NamedRerouteDeclaration(None=SphereMask(
    A=VertexInterpolator(VS=@n41).PS
    B=Constant() {r=0.5}
    Radius=ScalarParameter() {defaultValue=1, group=SideOn, parameterName=FacingMask_R}
    Hardness=LinearInterpolate(B=@n42, Alpha=@n43) {constA=0.75}
   )) {name=LightCone_FacingMask, nodeColor=<rgba 0 0.5781 1 1>}
  n44 = MaterialFunctionCall(Float2=TextureCoordinate()) {materialFunction=&BreakOutFloat2Components}
  Panner(Coordinate=MaterialFunctionCall(X=@n44.G, Y=@n44.G) {materialFunction=&MakeFloat2}.Result)
  n45 = LinearInterpolate(
   A=ScalarParameter() {parameterName=MotionBlurStr}
   B=ScalarParameter() {defaultValue=0.25, group=FrontEnd, parameterName=FrontEndMoBlurStr}
   Alpha=DynamicParameter() {paramNames [Param1,Param2,IsFrontEnd,Param4]}.IsFrontEnd
  )
  n46 = VectorParameter() {parameterName=MotionBlurStr, defaultValue=<rgba 0 0 50 1>}
  n47 = ComponentMask(None=TextureCoordinate()) {g=true}
  PreviousFrameSwitch(Current Frame=Constant3Vector(), Previous Frame=Multiply(A=Multiply(A=@n45, B=@n46.RGB), B=OneMinus(None=@n47)))
  ScalarParameter() {parameterName=TEST}
  TextureSample() {texture=/Game/Effects/Fort_Effects/Emotes/Sienna/T_Horizontal_gradient_Mask}
 ]
}
