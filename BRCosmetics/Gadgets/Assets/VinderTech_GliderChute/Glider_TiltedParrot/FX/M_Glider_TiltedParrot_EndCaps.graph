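# Translucent unlit Niagara mesh-particle material for the end-cap FX of the Tilted Parrot glider, visible mainly when viewed along the mesh's local X axis.
# Particle color tints a half-desaturated, low-mip glider diffuse for emissive.
# Panning noise and motion-blur streaks, shaped by a radial mask, particle alpha and the view-angle mask, drive opacity.
# Low-mask vertices collapse to the pivot, and an offset previous-frame position fakes velocity.
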Material /BRCosmetics/Gadgets/Assets/VinderTech_GliderChute/Glider_TiltedParrot/FX/M_Glider_TiltedParrot_EndCaps
graph {
 properties {
  bOutputTranslucentVelocity=true
  bTangentSpaceNormal=false
  bUsedWithNiagaraMeshParticles=true
  blendMode=BLEND_Translucent
  pixelDepthOffsetMode=PDOM_Legacy
  shadingModel=MSM_Unlit
  twoSided=true
  usedShadingModels=Unlit
 }
 nodes [
  n1 = Subtract(
   A=CameraPositionWS()
   B=MaterialFunctionCall() {materialFunction=&ObjectPivotPoint}.Mesh Particle Pivot Location
  )
  n2 = Transform(None=Constant3Vector() {constant=<rgba 1 0 0 0>}) {transformSourceType=TRANSFORMSOURCE_Local}
  n3 = Power(Base=DotProduct(A=Normalize(VectorInput=@n1), B=Normalize(VectorInput=@n2)), Exp=ScalarParameter() {parameterName=ex})
  Abs(None=@n3)
  Constant3Vector() {constant=<rgba 1 0 0 0>}
  Constant3Vector() {constant=<rgba 0 1 0 0>}
  n4 = NamedRerouteUsage()
  LinearInterpolate(A=Constant3Vector() {constant=<rgba 1 0 0 0>}, B=Constant3Vector() {constant=<rgba 0 1 0 0>}, Alpha=@n4)
  n5 = ParticleColor()
  n6 = TextureSample!(UVs=TextureCoordinate() {coordinateIndex=1}, Tex, Level, Apply View MipBias) {
   constMipValue=3
   mipValueMode=TMVM_MipLevel
   texture=/BRCosmetics/Gadgets/Assets/VinderTech_GliderChute/Glider_Sunrise_Castle_Male/Textures/T_Glider_Sunrise_Castle_Male_D
  }
  n7 = Constant() {r=0.5}
  MaterialFunctionCall(Emissive=Multiply(A=@n5.RGB, B=Desaturation(None=@n6.RGB, Fraction=@n7))) {materialFunction=&MF_FX_EmissiveAdjustments}
  n8 = VectorParameter() {parameterName=NoiseScale, defaultValue=<rgba 0 0 0 1>}
  n9 = Add(A=Time(), B=DynamicParameter() {paramNames [Param1,Param2,IsFrontEnd,Param4]}.Param2)
  n10 = Panner(Coordinate=Multiply(A=TextureCoordinate(), B=ComponentMask(None=@n8.RGB) {g=true, r=true}), Time=@n9, Speed=AppendVector(A=@n8.B, B=@n8.B)) {speedY=-3}
  n11 = ScalarParameter() {parameterName=Thicken}
  n12 = TextureCoordinate()
  n13 = Constant() {r=0.5}
  n14 = Multiply(
   A=Power(Base=SphereMask(A=@n12, B=@n13) {attenuationRadius=0.5, hardnessPercent=0}) {constExponent=1}
   B=ScalarParameter() {defaultValue=0.5, parameterName=Opacity}
  )
  n15 = Multiply(
   A=Add(
    A=TextureSample!(UVs=@n10, Tex, Apply View MipBias) {samplerType=SAMPLERTYPE_Grayscale, texture=&LowResBlurredNoise}.R
    B=Multiply(A=@n11, B=@n14)
   )
   B=@n14
  )
  n16 = TextureSample(UVs=Panner(Time=@n9) {speedY=-3}) {
   samplerType=SAMPLERTYPE_LinearColor
   texture=/BRCosmetics/Gadgets/Assets/VinderTech_GliderChute/Glider_Sunrise_Castle_Male/FX/T_FX_SunriseCastleMale_MotionBlurStreaks
  }
  n17 = Multiply(A=@n16.G, B=@n14)
  Multiply(A=@n5.A, B=Multiply(A=Add(A=@n15, B=@n17), B=@n4) {constB=32})
  NamedRerouteDeclaration(None=VertexInterpolator(VS=@n3).PS) {name=FrontOnMasking, nodeColor=<rgba 0 1 0.0859 1>}
  NamedRerouteDeclaration(None=@n3) {name=FrontOnMasking_VS, nodeColor=<rgba 1 0.0234 0 1>}
  n18 = LinearInterpolate(
   A=Subtract(
    A=MaterialFunctionCall() {materialFunction=&ObjectPivotPoint}.Mesh Particle Pivot Location
    B=WorldPosition().XYZ
   )
   Alpha=Step(X=NamedRerouteUsage()) {constY=0.01}
  ) {constB=0}
  PreviousFrameSwitch(Current Frame=@n18, Previous Frame=Add(A=@n18, B=Constant3Vector() {constant=<rgba 0 0 -10 0>}))
  WorldPosition()
 ]
}
