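# Idle Niagara effect for the TiltedParrot glider, adding ribbon trails and motion-blur meshes to its skeletal mesh.
# Two anim trails stream between bone or socket pairs, with a sine-wobbled velocity in the front end and alpha driven by system ForwardMovement.
# Two mesh emitters each pin four persistent particles to wing indices, stretching them by motion-blur weight.
# Their alpha follows a system timeline adjusted by time of day.
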NiagaraSystem /BRCosmetics/Gadgets/Assets/VinderTech_GliderChute/Glider_TiltedParrot/FX/NS_Glider_TiltedParrot_Idle
dependencies {
 usedDataInterfaces [&NiagaraDataInterfaceCurve]
 usedDynamicInputs [
  &RemapRange
  &VectorLength
  &Subtract_Float
  &DotProduct
  &NormalizeVector
  &Multiply_Vector
 ]
 usedModules [
  &SystemState
  &self:SystemSpawnScript.SystemScriptSource.SystemScriptGraph.NiagaraNodeAssignment_0.SetVariables_#1
  &self:SystemSpawnScript.SystemScriptSource.SystemScriptGraph.NiagaraNodeAssignment_1.SetVariables_#2
  &FN_Integrator
  &Timeline
 ]
}
emitters {
 FN_AnimTrail {
  emitterSpawnScript [&FN_RibbonIDOffsetUtility()]
  emitterUpdateScript [
   &EmitterState()
   &FN_SpawnAnimTrail(Number of Anim Trails=2)
  ]
  particleSpawnScript [
   &FN_InitializeAnimTrailBoneIndecies(First Bone or Socket Index=1, Second Bone or Socket Index=0)
   &FN_InitializeAnimTrailBoneIndecies(AnimTrail Selection=1, First Bone or Socket Index=3, Second Bone or Socket Index=2) {module=FN_InitializeAnimTrailBoneIndecies001}
   &FN_InitializeAnimTrail(Color=<rgba 0.147 0.2379 0 1>, Lifetime=0.15)
   &AddVelocity(Velocity=<0 400 0>) {enabled=false}
   &AddVelocity(Velocity=&FN_FrontEndVector3Modifier(
     Gameplay=<0 1250 0>
     Front End=&Multiply_VectorByFloat(
      Vector=&MakeVector(
       X=0
       Y=1428.8495
       Z=&Multiply_Float(A=&Sine(Normalized Angle=$Emitter.Age, Period=0.7, Scale=1, Bias=5.7178), B=60)
      )
      Float=1
     )
    )) {module=AddVelocity001}
  ]
  particleUpdateScript [
   &InheritSourceMovement(
    Applied Position Scale=&VectorFromFloat(Value=&FloatFromCurve(FloatCurve=DataInterface() {propertyValues {Curve {keys [<at 0 1 RCIM_Cubic>,<at 1 0 RCIM_Cubic>]}}}))
    ApplyForce=false
    ApplyPosition=true
    ApplyPositionFalloff=false
   ) {enabled=false}
   &SolveForcesAndVelocity(Clamp Velocity=false, Limit Acceleration=false)
   &FN_UpdateAnimTrail(Locked Particle RibbonWidth=1, SkeletalMesh=$User.SkeletalMesh) {module=FN_UpdateAnimTrail001}
   &ScaleColor(Scale Alpha=&Lerp_Float(A=0, B=1, Alpha=$System.ForwardMovement), ScaleA=true, ScaleRGB=true)
  ]
  rendererClasses [&NiagaraRibbonRendererProperties]
  renderers [
   {rendererClass=&NiagaraRibbonRendererProperties}
   {rendererClass=&NiagaraRibbonRendererProperties, rendererIndex=1}
  ]
 }
 MotionBlurMeshes {
  emitterUpdateScript [
   &EmitterState(
    Inactive Response=::NewEnumerator1
    Life Cycle Mode=::NewEnumerator1
    Loop Behavior=::NewEnumerator1
    Loop Duration Mode=::NewEnumerator1
    Scalability Mode=::NewEnumerator1
    UseLoopDelay=false
   )
   &SpawnBurst_Instantaneous(Spawn Count=4, Use Spawn Probability=false, UseLoopCountLimit=false)
  ]
  particleSpawnScript [
   &InitializeParticle(
    Color=<rgba 0.3505 1 0.1429 1>
    Lifetime=0
    Mesh Scale=<1 1 1.75>
    Mesh Scale Mode=::NewEnumerator4
    UsePositionOffset=false
    Write Material Random=true
   )
   /SetVariables(
    Particles.EndCap_Color:LinearColor=<rgba 0.2268 1 0 1>
    Particles.Joint_Color:LinearColor=<rgba 0.4605 1 0 0.5>
    Particles.Joint_Size:Vector3f=<1.5 1 1>
    Particles.WingIndex:NiagaraInt32=&ModuloInt(A=$Particles.UniqueID, B=4)
   ) {isSetParameters=true, module=SetVariables_#3}
   /Niagara/Modules/Update/Orientation/V2/OrientMeshToVector(
    Look At Direction:Vector3f=&MakeCustomVectorFromBool(
     Bool=&SetBoolByIntComparison(A=$Particles.WingIndex, B=1, Comparison Type=::NewEnumerator0)
     True Vector=<-1 0 0>
     False Vector=<1 0 0>
    )
    Override Quat=false
    Reference Axis Vector:Vector3f=<0 0 0>
    Side Axis Direction:Vector3f=&MakeCustomVectorFromBool(
     Bool=&SetBoolByIntComparison(A=$Particles.WingIndex, B=1, Comparison Type=::NewEnumerator0)
     True Vector=<0 -1 0>
     False Vector=<0 1 0>
    )
    Up Axis Direction:Vector3f=<0 0 0>
    UseOptionalSideAxis=true
    UseOptionalUpAxis=false
   ) {enabled=false}
   &InitialMeshOrientation(
    Mesh Orientation Mode=::NewEnumerator2
    Orientation Vector=&MakeCustomVectorFromBool(
     Bool=&SetBoolByIntComparison(A=$Particles.WingIndex, B=1, Comparison Type=::NewEnumerator0)
     True Vector=<-1 0 0>
     False Vector=<1 0 0>
    )
    Rotation=<0.5 0 0>
    Use Rotation Vector=true
   )
  ]
  particleUpdateScript [
   &ParticleState(Kill Particles When Lifetime Has Elapsed=false, Let Infinitely Lived Particles Die When Emitter Deactivates=true)
   &SkeletalMeshLocation(
    Bone / Socket Index=$Particles.WingIndex
    Bone Sampling Mode=::NewEnumerator4
    MaskSpawnGroup=false
    Sampled Position Offset=&MakeCustomVectorFromBool(
     Bool=&SetBoolByIntComparison(A=$Particles.WingIndex, B=1, Comparison Type=::NewEnumerator0)
     True Vector=<5 0 0>
     False Vector=<-5 0 0>
    )
    Skeletal Mesh=$User.SkeletalMesh
   )
   &DynamicMaterialParameters(
    Index 0 Param 1=$System.FX_MotionBlurWeight
    Index 0 Param 2=&RandomRangeFloat(Minimum=0, Maximum=3)
    Index 0 Param 3=&MakeCustomFloatFromBool(Bool=$User.bIsFrontEnd, True Float=1, False Float=0)
    Index 0 Param 4=1
    Module.Param0WriteEnabled=true
    Module.Param1WriteEnabled=true
    Module.Param2WriteEnabled=true
    Module.Param3WriteEnabled=true
   )
   &ScaleMeshSize(Scale Factor=&MakeVector(X=1, Y=1, Z=$System.FX_MotionBlurWeight))
   &ScaleColor(
    Scale Alpha=&FNTimeOfDayScalar(
     Night Scale=&Multiply_Float(A=$System.Timeline.Time, B=1)
     Morning Scale=$System.Timeline.Time
     Day Scale=$System.Timeline.Time
     Evening Scale=&Multiply_Float(A=$System.Timeline.Time, B=0.75)
    )
    Scale RGB=&FN_Select_Vector_from_Time_Of_Day(Morning=<1 1 1>, Day=<1 1 1>, Evening=<0 1 1>, Night=<0 1 1>)
    ScaleA=true
    ScaleRGB=true
   )
  ]
  rendererClasses [&NiagaraMeshRendererProperties]
  renderers [
   {rendererClass=&NiagaraMeshRendererProperties, rendererIndex=0}
   {rendererClass=&NiagaraMeshRendererProperties, rendererIndex=1}
  ]
 }
 MotionBlurMeshes001 {
  emitterUpdateScript [
   &EmitterState(
    Inactive Response=::NewEnumerator1
    Life Cycle Mode=::NewEnumerator1
    Loop Behavior=::NewEnumerator1
    Loop Duration Mode=::NewEnumerator1
    Scalability Mode=::NewEnumerator1
    UseLoopDelay=false
   )
   &SpawnBurst_Instantaneous(Spawn Count=4, Use Spawn Probability=false, UseLoopCountLimit=false)
  ]
  particleSpawnScript [
   &InitializeParticle(
    Color=<rgba 0.1754 0.6106 0.0437 0.5>
    Lifetime=0
    Mesh Scale=<1.5 1 1>
    Mesh Scale Mode=::NewEnumerator4
    UsePositionOffset=false
    Write Material Random=true
   )
   /SetVariables(Particles.WingIndex:NiagaraInt32=&ModuloInt(A=$Particles.UniqueID, B=4)) {isSetParameters=true, module=SetVariables_#3}
   &OrientMeshToVector(
    Look At Direction:Vector3f=&MakeCustomVectorFromBool(
     Bool=&SetBoolByIntComparison(A=$Particles.WingIndex, B=1, Comparison Type=::NewEnumerator0)
     True Vector=<-1 0 0>
     False Vector=<1 0 0>
    )
    Override Quat=false
    Reference Axis Vector:Vector3f=<0 0 0>
    Side Axis Direction:Vector3f=&MakeCustomVectorFromBool(
     Bool=&SetBoolByIntComparison(A=$Particles.WingIndex, B=1, Comparison Type=::NewEnumerator0)
     True Vector=<0 -1 0>
     False Vector=<0 1 0>
    )
    Up Axis Direction:Vector3f=<0 0 0>
    UseOptionalSideAxis=true
    UseOptionalUpAxis=false
   ) {enabled=false}
   &InitialMeshOrientation(
    Mesh Orientation Mode=::NewEnumerator2
    Orientation Vector=&MakeCustomVectorFromBool(
     Bool=&SetBoolByIntComparison(A=$Particles.WingIndex, B=1, Comparison Type=::NewEnumerator0)
     True Vector=<-1 0 0>
     False Vector=<1 0 0>
    )
    Rotation=<0.5 0 0>
    Use Rotation Vector=true
   )
  ]
  particleUpdateScript [
   &ParticleState(Kill Particles When Lifetime Has Elapsed=false, Let Infinitely Lived Particles Die When Emitter Deactivates=true)
   &SkeletalMeshLocation(
    Bone / Socket Index=$Particles.WingIndex
    Bone Sampling Mode=::NewEnumerator4
    MaskSpawnGroup=false
    Sampled Position Offset=&MakeCustomVectorFromBool(
     Bool=&SetBoolByIntComparison(A=$Particles.WingIndex, B=1, Comparison Type=::NewEnumerator0)
     True Vector=<15 0 0>
     False Vector=<-14 0 0>
    )
    Skeletal Mesh=$User.SkeletalMesh
   )
   &DynamicMaterialParameters(
    Index 0 Param 1=$System.FX_MotionBlurWeight
    Index 0 Param 2=&RandomRangeFloat(Minimum=0, Maximum=3)
    Index 0 Param 3=&MakeCustomFloatFromBool(Bool=$User.bIsFrontEnd, True Float=1, False Float=0)
    Index 0 Param 4=1
    Module.Param0WriteEnabled=true
    Module.Param1WriteEnabled=true
    Module.Param2WriteEnabled=true
    Module.Param3WriteEnabled=true
   )
   &ScaleMeshSize(Scale Factor=&MakeVector(X=1, Y=1, Z=$System.FX_MotionBlurWeight))
   &ScaleColor(
    Scale Alpha=&FNTimeOfDayScalar(
     Night Scale=&Multiply_Float(A=$System.Timeline.Time, B=0.75)
     Morning Scale=$System.Timeline.Time
     Day Scale=$System.Timeline.Time
     Evening Scale=&Multiply_Float(A=$System.Timeline.Time, B=0.75)
    )
    ScaleA=true
    ScaleRGB=true
   )
  ]
  rendererClasses [&NiagaraMeshRendererProperties]
  renderers [{rendererClass=&NiagaraMeshRendererProperties, rendererIndex=0}]
 }
}
summary {
 systemName=NS_Glider_TiltedParrot_Idle
 emitters [
  {emitterName=FN_AnimTrail, rendererClasses [&NiagaraRibbonRendererProperties]}
  {emitterName=MotionBlurMeshes, rendererClasses [&NiagaraMeshRendererProperties]}
  {emitterName=MotionBlurMeshes001, rendererClasses [&NiagaraMeshRendererProperties]}
 ]
 userVariables [
  User.bIsFrontEnd {type=NiagaraBool}
  User.bIsFrontEndPreview {type=NiagaraBool}
  User.bisFullyDeployed {type=NiagaraBool}
  User.ForwardDot {type=NiagaraFloat, defaultValue=0, description=""}
  User.RibbonIDOffset {type=NiagaraInt32}
  User.SkeletalMesh {
   type=NiagaraDataInterfaceSkeletalMesh
   defaultValue {
    struct=NiagaraExt_VariableValue_DataInterface
    value {dataInterface=&self:NiagaraDataInterfaceSkeletalMesh_0}
   }
  }
  User.Speed {type=NiagaraFloat, defaultValue=0, description=""}
 ]
}
